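; Latent burst gameplay cue that plays a ping sound, named for marking an opponent in the TDM mode of the Flex Legend Medallion feature.
; On burst, it reads the cue's raw magnitude as a delay and sets a timer to PlaySound when positive, otherwise plays immediately.
; The sound attaches to the cue actor's own root component rather than MyTarget, and debug prints trace the burst and playback.
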(fn OnBurst(EitherVFX-preferthisversion)
  (params (MyTarget :type Actor%20Object%20Reference) (Parameters :type Gameplay%20Cue%20Parameters%20Structure%20%28by%20ref%29) (ParticleComponents :type Array%20of%20FXSystem%20Component%20Object%20References) (AudioComponents :type Array%20of%20Audio%20Component%20Object%20References) (BurstCameraShakeInstance :type Camera%20Shake%20Base%20Object%20Reference) (BurstDecalInstance :type Decal%20Actor%20Object%20Reference))
  (bind (_normalizedmagnitude _rawmagnitude) (Ability|GameplayCue|BreakGameplayCueParameters Parameters))
  (Development|PrintString 0 "MARKING_CUE: On Burst ( ) " true true "(R=0.494069,G=0.000000,B=1.000000,A=1.000000)" 8.0)
  (if (> _rawmagnitude 0)
    (Utilities|Time|SetTimerbyEvent (EventDispatchers|CreateEvent :!event PlaySound) _rawmagnitude)
    (return)
    (else
      (Audio|SpawnSoundAttached (Variables|Default|GetPingSound) (Variables|Transformation|GetRootComponent))
      (Development|PrintString 0 "MARKING_CUE: Play Sound ( ) " true true "(R=0.494069,G=0.000000,B=1.000000,A=1.000000)" 8.0))))

(fn PlaySound
  (Audio|SpawnSoundAttached (Variables|Default|GetPingSound) (Variables|Transformation|GetRootComponent))
  (Development|PrintString 0 "MARKING_CUE: Play Sound ( ) " true true "(R=0.494069,G=0.000000,B=1.000000,A=1.000000)" 8.0))

(fn ConstructionScript)
